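# Translucent glass material for the Modern Witch pickaxe, blending a noisy, depth-faded water layer into tinted glass with a boosted Fresnel mask.
# Preskinned-position coordinates drive the scrolling noise and normals, while a time-driven sine pushes vertices along their normals.
# Noise and camera distance drive refraction, which, like the normal detail, flattens on Low and Medium quality.
# Only six attributes reach the TechArt override function; blended opacity, metallic, specular and roughness are not forwarded.
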Material /BRCosmetics/Weapons/FORT_Melee/Pickaxe_Modern_Witch/Materials/M_Pickaxe_Modern_Witch_Glass
graph {
 properties {
  bCastRayTracedShadows=false
  bEnableResponsiveAA=true
  bScreenSpaceReflections=true
  bUseMaterialAttributes=true
  bUsedWithMeshDeformer=true
  bUsedWithSkeletalMesh=true
  blendMode=BLEND_Translucent
  ditherOpacityMask=true
  physMaterial=/Game/FortressPhysicalMaterials/Glass
  pixelDepthOffsetMode=PDOM_Legacy
  refractionMethod=RM_IndexOfRefraction
  translucencyDirectionalLightingIntensity=5
  translucencyLightingMode=TLM_VolumetricDirectional
  usedShadingModels=Default Lit
 }
 nodes [
  Constant()
  n1 = VectorParameter() {parameterName=waterColor, defaultValue=<rgba 0.1426 0.2846 0.3802 1>}
  n2 = Clamp(None=Fresnel() {baseReflectFraction=0, exponent=1})
  n3 = TextureObject() {samplerType=SAMPLERTYPE_Grayscale, texture=&LowResBlurredNoise}
  n4 = ScalarParameter() {parameterName=Deform Speed X}
  n5 = ScalarParameter() {defaultValue=0.1, parameterName=Deform Speed Y_1}
  n6 = MaterialFunctionCall(Float3=MaterialFunctionCall!(
    TextureObject=@n3
    TextureSize=ScalarParameter() {defaultValue=5, parameterName=DeformSize_A}
    WorldPosition
    Export Float 4
    In=AppendVector(A=@n4, B=@n5)
    World Space Normal
    ProjectionTransitionContrast
   ) {materialFunction=/Game/Athena/SafeZone/Preskinned_WP_Texture_ScalePan}.XYZ Texture) {materialFunction=&BreakOutFloat3Components}
  n7 = ScalarParameter() {defaultValue=0.2, parameterName=Deform Speed Y_2}
  n8 = MaterialFunctionCall(Float3=MaterialFunctionCall(TextureObject=@n3, TextureSize=ScalarParameter() {defaultValue=10, parameterName=DeformSize_A_1}, In=AppendVector(A=@n4, B=@n7)) {materialFunction=&Preskinned_WP_Texture_ScalePan}.XYZ Texture) {materialFunction=&BreakOutFloat3Components}
  n9 = Saturate(None=Multiply(A=OneMinus(None=@n2), B=Multiply(A=@n6.R, B=@n8.R)))
  n10 = ScalarParameter() {defaultValue=0.6, parameterName=Opac}
  n11 = ScalarParameter() {defaultValue=9.5587, parameterName=FadeDistance}
  n12 = Add(A=Multiply(A=@n9, B=@n9), B=QualitySwitch(Default=DepthFade(Opacity=@n10, FadeDistance=@n11), Low=Constant()))
  n13 = VectorParameter() {parameterName=Glass Base Color, defaultValue=<rgba 0.4889 1.25 0.1823 1>}
  n14 = ScalarParameter() {defaultValue=0.1, parameterName=Metallic}
  n15 = ScalarParameter() {defaultValue=1, parameterName=Specular}
  n16 = ScalarParameter() {defaultValue=0.1, parameterName=Roughness}
  n17 = Multiply(A=@n13.RGB) {constB=0.1}
  n18 = PreSkinnedPosition()
  n19 = MaterialFunctionCall(Float3=Divide(A=VertexInterpolator(VS=@n18).PS, B=ObjectBounds())) {materialFunction=&BreakOutFloat3Components}
  n20 = Frac(None=Multiply(A=Time()) {constB=0.1})
  n21 = Multiply(A=AppendVector(A=@n19.G, B=Add(A=@n19.B, B=@n20)), B=ScalarParameter() {defaultValue=0.4, parameterName=WaterTiling})
  n22 = Normalize(VectorInput=Multiply(
    A=TextureSample(UVs=@n21) {samplerType=SAMPLERTYPE_Normal, texture=/Game/Global/Textures/Default/Noise/T_Noise_Normal_01}.RGB
    B=VectorParameter() {parameterName=Normal Intensity, defaultValue=<rgba 0.1 0.1 1 0>}.RGB
   ))
  n23 = LinearInterpolate(
   A=ScalarParameter() {parameterName=Fresnel Inner Transparency}
   B=ScalarParameter() {defaultValue=1, parameterName=Fresnel Outer Transparency}
   Alpha=Fresnel(ExponentIn=ScalarParameter() {defaultValue=4, parameterName=Fresnel Exponent}, Normal=Transform(None=@n22))
  )
  n24 = Constant3Vector() {constant=<rgba 0 0 1 0>}
  n25 = QualitySwitch(Default=@n22, Low=@n24, High=@n22, Medium=@n24)
  n26 = ScalarParameter() {defaultValue=3.9792, parameterName=BowlThickness}
  n27 = VertexNormalWS()
  n28 = Clamp(None=Multiply(A=Fresnel(ExponentIn=@n26, Normal=@n27) {baseReflectFraction=0}, B=ScalarParameter() {defaultValue=4.8435, parameterName=Boost}))
  n29 = ScalarParameter() {defaultValue=0.7953, parameterName=TransitionPower}
  n30 = BreakMaterialAttributes!(Attr=BlendMaterialAttributes(
    A=MakeMaterialAttributes!(
     BaseColor=@n1.RGB
     Metallic
     Specular
     Roughness
     Anisotropy
     EmissiveColor
     Opacity=@n12
     OpacityMask
     Normal
     Tangent
     WorldPositionOffset
     SubsurfaceColor
     ClearCoat
     ClearCoatRoughness
     AmbientOcclusion
     Refraction
     CustomizedUVs_0
     CustomizedUVs_1
     CustomizedUVs_2
     CustomizedUVs_3
     CustomizedUVs_4
     CustomizedUVs_5
     CustomizedUVs_6
     CustomizedUVs_7
     PixelDepthOffset
     ShadingModel
     Displacement
    )
    B=MakeMaterialAttributes(BaseColor=@n13.RGB, Metallic=@n14, Specular=@n15, Roughness=@n16, EmissiveColor=@n17, Opacity=@n23, Normal=@n25)
    Alpha=Power(Base=@n28, Exp=@n29) {constExponent=14}
   ) {vertexAttributeBlendType=UseB})
  n31 = MaterialFunctionCall(Float3=Multiply(A=@n18, B=ScalarParameter() {defaultValue=0.02, parameterName=WPO_LocalNoise_Size}) {constB=0.02}) {materialFunction=&BreakOutFloat3Components}
  n32 = Subtract(A=Add(A=Add(A=@n31.R, B=@n31.R), B=@n31.B)) {constB=2}
  n33 = Time()
  n34 = ScalarParameter() {parameterName=WPO_random_seed}
  n35 = Frac(None=Multiply(A=Subtract(A=@n33, B=@n34), B=ScalarParameter() {defaultValue=0.6, parameterName=WPO_LocalNoise_Speed}) {constB=0.6})
  n36 = Add(A=Sine(None=Subtract(A=@n32, B=@n35)))
  n37 = Multiply(A=Multiply(A=VertexNormalWS(), B=Multiply(A=@n36) {constB=0.5}), B=ScalarParameter() {defaultValue=1, parameterName=WPO_FinalScale})
  n38 = CameraPositionWS()
  n39 = WorldPosition() {worldPositionShaderOffset=WPT_ExcludeAllShaderOffsets}
  n40 = Saturate(None=Multiply(A=Distance(A=@n38, B=@n39.XYZ), B=ScalarParameter() {defaultValue=0.001, parameterName=Refraction Distance Scale}) {constB=0.002})
  n41 = ScalarParameter() {defaultValue=2, parameterName=Refraction Distance Power}
  n42 = LinearInterpolate(
   A=ScalarParameter() {defaultValue=1, parameterName=Refraction_2, sliderMax=2}
   B=ScalarParameter() {defaultValue=0.8, parameterName=Refraction_Distort, sliderMax=2}
   Alpha=@n9
  )
  n43 = Saturate(None=Add(A=Power(Base=@n40, Exp=@n41) {constExponent=2.5}, B=Clamp(None=@n42)))
  n44 = Constant() {r=1}
  n45 = SetMaterialAttributes!(
   MaterialAttributes
   Base Color=NamedRerouteDeclaration(None=@n30.BaseColor) {name=baseColor, nodeColor=<rgba 1 0 0.7969 1>}
   Emissive Color=NamedRerouteDeclaration(None=@n30.EmissiveColor) {name=emissive, nodeColor=<rgba 1 0.4453 0 1>}
   Normal=NamedRerouteDeclaration(None=@n30.Normal) {name=nmormal, nodeColor=<rgba 0 0.5781 1 1>}
   World Position Offset=NamedRerouteDeclaration(None=Add(A=@n37, B=@n30.WorldPositionOffset)) {name=wpo, nodeColor=<rgba 1 0.7266 0 1>}
   Ambient Occlusion=NamedRerouteDeclaration(None=@n30.AmbientOcclusion) {name=ambocc, nodeColor=<rgba 0.7578 1 0 1>}
   Refraction=NamedRerouteDeclaration(None=QualitySwitch(Default=@n43, Low=@n44, High=@n43, Medium=@n44)) {name=refraction, nodeColor=<rgba 0 0.4844 1 1>}
  ) {attributeSetTypes [
    #1
    #2
    #3
    #4
    #5
    #6
   ]}
  MaterialFunctionCall(In=@n45) {materialFunction=&MF_TechArtOverrides_accessories}
  NamedRerouteDeclaration(None=@n30.Opacity) {name=opacity, nodeColor=<rgba 0.8984 1 0 1>}
 ]
}
