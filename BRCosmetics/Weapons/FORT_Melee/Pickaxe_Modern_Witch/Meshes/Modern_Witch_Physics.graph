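# Physics asset for the Modern Witch two-handed pickaxe, giving collision bodies and joints to its handle and "dyn" ring bones.
# A long capsule covers the handle, while a sphere and smaller capsules cover the inner and outer rings.
# Each of the three ring bodies is constrained directly to the handle rather than chained.
# The joints differ per ring, mixing free and locked angular axes so each moves differently relative to the handle.
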PhysicsAsset /BRCosmetics/Weapons/FORT_Melee/Pickaxe_Modern_Witch/Meshes/Modern_Witch_Physics
bodies {
 Melee_TwoHanded_Handle [{shapeType=Capsule, length=35, radius=3, center=<0.3378 -1.2983 46.1861>}]
 dyn_inner_ring_01 [{shapeType=Sphere, radius=8, center=<0.1 0 0>}]
 dyn_outer_ring [{shapeType=Capsule, length=5, radius=5, center=<1 2.8 0>, rotation=<rot 90 0 0>}]
 dyn_outer_ring_01 [{shapeType=Capsule, length=0, radius=15, center=<0.1 0 0>}]
}
constraints [
 {
  bone1Name=dyn_inner_ring_01
  bone2Name=Melee_TwoHanded_Handle
  swing1LimitDegrees=90
  swing1Motion=Free
  swing2LimitDegrees=90
  swing2Motion=Free
  twistLimitDegrees=5
  twistMotion=Locked
 }
 {bone1Name=dyn_outer_ring, bone2Name=Melee_TwoHanded_Handle, swing1LimitDegrees=45, swing2LimitDegrees=1, twistLimitDegrees=1}
 {
  bone1Name=dyn_outer_ring_01
  bone2Name=Melee_TwoHanded_Handle
  swing1LimitDegrees=45
  swing1Motion=Locked
  swing2LimitDegrees=45
  swing2Motion=Free
  twistLimitDegrees=20
 }
]
