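# Animation Blueprint for the Modern Witch pickaxe that continuously spins its inner and outer rings, speeding them up with the wielder's hand motion.
# Each update derives an eased velocity from the attach parent's hand_r and spine_05 sockets.
# Spin Function adds base speed plus scaled velocity to two wrapped angles, applied as additive bone-space rotations.
# The RigidBody node and SpinX logic are not wired into the active chains.
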AnimBlueprint /BRCosmetics/Weapons/FORT_Melee/Pickaxe_Modern_Witch/Meshes/Modern_Witch_AnimBP
target_skeleton=&/Modern_Witch_Skeleton
graphs [
 AnimGraph {nodes [
   RigidBody(ComponentPose=Local To Component(LocalPose=Input Pose())) {
    bClampLinearTranslationLimitToRefPose=true
    bOverrideWorldGravity=true
    overridePhysicsAsset=&/Modern_Witch_Physics
    alphaBoolBlend {blendOption=Linear}
    baseBoneRef {boneName=Melee_TwoHanded_Handle}
    componentAppliedLinearAccClamp=<250 250 250>
    componentLinearAccScale=<0.5 0.5 0.5>
    componentLinearVelScale=<0.5 0.5 0.5>
    overrideWorldGravity {z=-980}
   }
   n1 = "Transform (Modify) Bone - Bone: dyn_inner_ring_02"(Alpha=1, Translation="0,0,0", Rotation=Get spinResultA() -> (spinResultA:Rotator), Scale="1,1,1") {rotationMode=BMM_Additive, rotationSpace=BCS_BoneSpace, alphaBoolBlend {blendOption=Linear}, boneToModify {boneName=dyn_inner_ring_02}}
   n2 = "Transform (Modify) Bone - Bone: dyn_outer_ring_02"(ComponentPose=@n1, Alpha=1, Translation="0,0,0", Rotation=Get spinResultB() -> (spinResultB:Rotator), Scale="1,1,1") {rotationMode=BMM_Additive, rotationSpace=BCS_BoneSpace, alphaBoolBlend {blendOption=Linear}, boneToModify {boneName=dyn_outer_ring_02}}
   Output Pose(Result=Component To Local(ComponentPose=@n2)) {layerGroup=None}
   Get velocity() -> (velocity:double)
  ]}
 EventGraph {nodes [
   Event BlueprintUpdateAnimation()
   > n1 = Sequence()
   > Set deltaTimeRemap(deltaTimeRemap:double=FInterpTo(Current=Get deltaTimeRemap() -> (deltaTimeRemap:double), Target=1, DeltaTime=0, InterpSpeed=1)) -> (Output_Get:double)
   n2 = GetPartSkeletalMeshComponent()
   Is Valid(exec=@n1.then_1, InputObject=@n2)
   n3 = GetAttachParent!(self:Scene Component Object Reference=@n2) -> (ReturnValue:Scene Component Object Reference)
   > Is Valid(InputObject=@n3)
   n4 = GetSocketTransform(self=@n3, InSocketName=hand_r, TransformSpace=RTS_World)
   n5 = GetSocketTransform(self=@n3, InSocketName=spine_05, TransformSpace=RTS_World)
   > n6 = Calculate Velocity From Position History!(
    execute
    DeltaSeconds:float=0.001
    Position:Vector=BreakTransform(InTransform=MakeRelativeTransform!(A:"Transform (by ref)"=@n5, RelativeTo:"Transform (by ref)"=@n4) -> (ReturnValue:Transform))
    History:"Position History Structure (by ref)"=Make Position History!() -> (PositionHistory:Position History Structure) {is=title K2Node_MakeStruct}
    NumberOfSamples:int=150
    VelocityMin:float=0.f
    VelocityMax:float=1800
   ) -> (then, ReturnValue:float)
   > Set velocity(velocity:double=Ease!(Function:EEasingFunc Enum=EaseInOut, Alpha:double=@n6, A:double, B:double=1, BlendExp:double=1) -> (Result:double=~) {is=title K2Node_EaseFunction}) -> (Output_Get:double)
   > n7 = Sequence!(execute) -> (then_0, then_1, then_2:Exec)
   > n8 = Spin Function!(execute, self, Spin:double=Get spinA() -> (spinA:double), Speed:double=Get speedA() -> (speedA:double)) -> (then, Output:double)
   > n9 = Set spinA(spinA:double=@n8) -> (Output_Get:double)
   > Set spinResultA(spinResultA:Rotator=MakeRotator(Pitch=@n9)) -> (Output_Get:Rotator)
   n10 = Spin Function(execute=@n7.then_1, Spin=Get spinB() -> (spinB:double), Speed=Get speedB() -> (speedB:double))
   > n11 = Set spinB(spinB:double=@n10) -> (Output_Get:double)
   > Set spinResultB(spinResultB:Rotator=MakeRotator(Pitch=@n11, Yaw=20)) -> (Output_Get:Rotator)
   n12 = Branch(Condition="float >= float"(A=Get SpinX() -> (SpinX:double), B=720) {is=title K2Node_CallFunction})
   > n13 = Set SpinX(SpinX:double=float - float(A=Get SpinX(), B=720) {is=title K2Node_CallFunction}) -> (Output_Get:double)
   n14 = Get SpinX()
   Set SpinConst(execute=[@n13.then, @n15.then], SpinConst:Rotator=MakeRotator(Roll=@n14, Pitch=float * float(A=@n14, B=1.5))) -> (Output_Get:Rotator)
   n15 = Set SpinX(execute=@n12.else, SpinX=float + float(A=Get SpinX(), B=3) {is=title K2Node_CommutativeAssociativeBinaryOperator})
   TryGetPawnOwner()
  ]}
 Spin Function {nodes [
   n1 = Spin Function!() -> (then, Spin:double=~, Speed:double=~) {is=title K2Node_FunctionEntry}
   > n2 = Set spinVar(spinVar:double=@n1) -> (Output_Get:double)
   > n3 = Branch(Condition="float >= float"(A=@n2, B=720))
   > n4 = Set spinVar(spinVar=float - float(A=Get spinVar() -> (spinVar:double), B=720))
   n5 = Get velocity()
   n6 = Get velocityMult() -> (velocityMult:double)
   n7 = Get deltaTimeRemap()
   n8 = Set spinVar(
    execute=@n3.else
    spinVar=float + float!(A=Get spinVar(), B=float * float(A=@n1.Speed, B=@n7), C:double=float * float(A=@n5, B=@n6)) -> (ReturnValue)
   )
   Return Node!(execute=[@n4.then, @n8.then], Output:double=@n8)
  ]}
]
variables [
 deltaTimeRemap {type=number, default=1}
 speedA {type=number, default=3}
 speedB {type=number, default=6}
 spinA {type=number}
 spinB {type=number}
 SpinConst {type=Rotator}
 spinResultA {type=Rotator}
 spinResultB {type=Rotator}
 spinVar {type=number}
 SpinX {type=number}
 velocity {type=number}
 velocityMult {type=number, default=6}
]
